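# Conversation database for The Paradigm NPC in the NPC library, defining the greeting dialogue shown when a player talks to the character.
# The entry tag leads to one speech node with a default line, a first-ever-conversation line, and higher-weighted lines gated on the player wearing specific character skins.
# After the greeting, link nodes branch into the shared NPC Quests and BasicServices conversations.
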ConversationDatabase /NPCLibrary/NPCs/TheParadigm/TheParadigm/NPCLibrary_Conversation_TheParadigm
entry_points [{tag=Conversation.Entry.TheParadigm, destinations [#1]}]
nodes [
 ConversationEntryPointNode_7 {
  id=#1
  class=ConversationEntryPointNode
  outputs [#2]
  properties {entryTag {tagName=Conversation.Entry.TheParadigm}}
 }
 Speech_C_4 {
  id=#2
  class=Speech_C
  outputs [#3, #4]
  properties {
   nodeName=""
   generalConfig {
    defaultMessage=You need something?
    contextualMessages [
     {
      message=Hey co-pilot.
      contextRequirements [{requirement {class=FortControllerRequirement_IsFirstEverConversationWithNPC}}]
      priority {curve {rowName=Tandem.Speech.Priority.FirstEverInteraction}}
     }
     {
      message=You can relax now. We won.
      contextRequirements [{requirement {
         class=FortControllerRequirement_HasCID
         properties {allowedCIds [/BRCosmetics/Athena/Items/Cosmetics/Characters/CID_A_373_Athena_Commando_M_OriginPrisoner]}
        }}]
      priority {value=10, curve {curveTable=null, rowName=None}}
      weight {value=10}
     }
     {
      message="If you're saying something's up, I believe you."
      contextRequirements [{requirement {
         class=FortControllerRequirement_HasCID
         properties {allowedCIds [/BRCosmetics/Athena/Items/Cosmetics/Characters/CID_540_Athena_Commando_M_MeteorManRemix]}
        }}]
      priority {value=10, curve {curveTable=null, rowName=None}}
      weight {value=10}
     }
     {
      message=Thanks for listening when no one else would.
      contextRequirements [{requirement {
         class=FortControllerRequirement_HasCID
         properties {allowedCIds [/BRCosmetics/Athena/Items/Cosmetics/Characters/CID_A_374_Athena_Commando_F_Binary]}
        }}]
      priority {value=10, curve {curveTable=null, rowName=None}}
      weight {value=10}
     }
     {
      message=Let me guess. You're just following orders?
      contextRequirements [{requirement {
         class=FortControllerRequirement_HasCID
         properties {allowedCIds [/BRCosmetics/Athena/Items/Cosmetics/Characters/CID_A_395_Athena_Commando_F_BinaryTwin]}
        }}]
      priority {value=10, curve {curveTable=null, rowName=None}}
      weight {value=10}
     }
     {
      message=It's good to be back.
      contextRequirements [{requirement {
         class=FortControllerRequirement_HasCID
         properties {allowedCIds [/BRCosmetics/Athena/Items/Cosmetics/Characters/CID_A_286_Athena_Commando_M_Turtleneck]}
        }}]
      priority {value=10, curve {curveTable=null, rowName=None}}
      weight {value=10}
     }
     {
      message=You look happy.
      contextRequirements [{requirement {
         class=FortControllerRequirement_HasCID
         properties {allowedCIds [/BRCosmetics/Athena/Items/Cosmetics/Characters/CID_140_Athena_Commando_M_Visitor]}
        }}]
      priority {value=10, curve {curveTable=null, rowName=None}}
      weight {value=10}
     }
     {
      message=I'm still not calling you The Legend.
      contextRequirements [{requirement {
         class=FortControllerRequirement_HasCID
         properties {allowedCIds [&CID_883_Athena_Commando_M_ChOneJonesy]}
        }}]
      priority {value=10, curve {curveTable=null, rowName=None}}
      weight {value=10}
     }
    ]
   }
  }
 }
 ConversationLinkNode_0 {id=#4, class=ConversationLinkNode, properties {remoteEntryTag {tagName=Conversation.Entry.BasicServices}}}
 ConversationLinkNode_1 {id=#3, class=ConversationLinkNode, properties {remoteEntryTag {tagName=Conversation.Entry.NPCLibrary.Quests}}}
]
